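# Physics asset for the Hero vehicle body cosmetic mesh, giving it a simplified collision shape.
# It has a single body on the root bone made of one large box, offset forward and upward from the bone origin.
# There are no constraints, so the whole body acts as one rigid collision volume.
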PhysicsAsset /VehicleCosmetics/Bodies/Hero/Meshes/Body_Hero_PhysicsDM
bodies {root [{shapeType=Box, extentX=444, extentY=216, extentZ=116, shapeName=root_box, center=<48 0 84>}]}
constraints []
